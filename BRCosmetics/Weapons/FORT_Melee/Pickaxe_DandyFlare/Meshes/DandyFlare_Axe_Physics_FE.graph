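# Physics asset for the DandyFlare two-handed pickaxe, simulating a flexible tube on the axe mesh.
# A handle capsule anchors a serial chain of seven tube bodies, capsules along the length and a sphere at the tip.
# Tight swing limits and zero twist keep the chain fairly stiff, with one narrower axis near the end.
# The last joint swings and twists freely, so the tip moves loosely.
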PhysicsAsset /BRCosmetics/Weapons/FORT_Melee/Pickaxe_DandyFlare/Meshes/DandyFlare_Axe_Physics_FE
bodies {
 Melee_TwoHanded_Handle [{shapeType=Capsule, length=64.6868, radius=5.8627, center=<-0.1734 -0.3497 13.7947>, rotation=<rot 1.3379 -90 0.8767>}]
 Tube_01 [{
   shapeType=Capsule
   length=3.6125
   radius=3.054
   shapeName=Tube_01_capsule
   center=<5.8853 0 0.5796>
   rotation=<rot 84.3751 -179.9993 -179.9992>
  }]
 Tube_02 [{shapeType=Capsule, length=6, radius=3, shapeName=Tube_02_capsule, center=<5.4459 0.3074 0>, rotation=<rot -90 5.6247 0>}]
 Tube_03 [{shapeType=Capsule, length=7, radius=3, shapeName=Tube_03_capsule, center=<6.6664 0 0.1377>, rotation=<rot -87.1876 179.9991 -179.9996>}]
 Tube_04 [{shapeType=Capsule, length=7, radius=3, center=<5.9314 0 0>, rotation=<rot 90 -0.0007 0>}]
 Tube_05 [{shapeType=Capsule, length=7, radius=3, shapeName=Tube_05_capsule, center=<5.3215 0 0>, rotation=<rot -90 0.0002 0>}]
 Tube_06 [{
   shapeType=Capsule
   length=7
   radius=3
   shapeName=Tube_06_capsule
   center=<5.6944 -0.0001 -0.2798>
   rotation=<rot -87.1875 -179.9989 179.9982>
  }]
 Tube_07 [{shapeType=Sphere, radius=3.0438, shapeName=Tube_07_sphere, center=<1.9553 0.3221 -0.9214>}]
}
constraints [
 {bone1Name=Tube_04, bone2Name=Tube_03, swing1LimitDegrees=6, swing2LimitDegrees=6, twistLimitDegrees=0}
 {bone1Name=Tube_01, bone2Name=Melee_TwoHanded_Handle, swing1LimitDegrees=5, swing2LimitDegrees=5, twistLimitDegrees=0}
 {bone1Name=Tube_02, bone2Name=Tube_01, swing1LimitDegrees=6, swing2LimitDegrees=6, twistLimitDegrees=0}
 {bone1Name=Tube_03, bone2Name=Tube_02, swing1LimitDegrees=6, swing2LimitDegrees=6, twistLimitDegrees=0}
 {bone1Name=Tube_06, bone2Name=Tube_05, swing1LimitDegrees=1, swing2LimitDegrees=6, twistLimitDegrees=0}
 {bone1Name=Tube_05, bone2Name=Tube_04, swing1LimitDegrees=6, swing2LimitDegrees=6, twistLimitDegrees=0}
 {
  bone1Name=Tube_07
  bone2Name=Tube_06
  swing1LimitDegrees=45
  swing1Motion=Free
  swing2LimitDegrees=45
  swing2Motion=Free
  twistLimitDegrees=45
  twistMotion=Free
 }
]
