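# Animation Blueprint for the DandyFlare pickaxe that adds rigid body secondary motion to the incoming pose.
# The cached input pose feeds three RigidBody nodes with separate physics assets and tuning: in-game, front end on a pawn, and front end without one.
# Bool blends select the result from bIsInFrontEnd and FE_withpawn, which is updated each tick.
# Both in-game and on-pawn front-end variants simulate relative to the two-handed handle bone.
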AnimBlueprint /BRCosmetics/Animation/Props/Gadgets_DandyFlare/Meshes/DandyFlare_Axe_SRS_AnimBP
target_skeleton=&/DandyFlare_Gadget_Skeleton
graphs [
 AnimGraph {nodes [
   n1 = RigidBody(ComponentPose=Local To Component(LocalPose=Use cached pose 'Base'())) {
    bClampLinearTranslationLimitToRefPose=true
    bOverrideWorldGravity=true
    overridePhysicsAsset=/BRCosmetics/Weapons/FORT_Melee/Pickaxe_DandyFlare/Meshes/DandyFlare_Axe_Physics_FE_without_pawn
    componentAppliedLinearAccClamp=<3000 3000 4000>
    overrideWorldGravity {z=-100}
    simSpaceSettings {dampingAlpha=0, maxAngularAcceleration=15, maxAngularVelocity=3, maxLinearAcceleration=1000, maxLinearVelocity=450, worldAlpha=1}
   }
   n2 = Component To Local(ComponentPose=@n1)
   n3 = RigidBody(ComponentPose=Local To Component(LocalPose=Use cached pose 'Base'())) {
    bClampLinearTranslationLimitToRefPose=true
    bOverrideWorldGravity=true
    overridePhysicsAsset=/BRCosmetics/Weapons/FORT_Melee/Pickaxe_DandyFlare/Meshes/DandyFlare_Axe_Physics_FE
    simulationSpace=BaseBoneSpace
    baseBoneRef {boneName=Melee_TwoHanded_Handle}
    componentAppliedLinearAccClamp=<3000 3000 3000>
    overrideWorldGravity {z=-800}
    simSpaceSettings {dampingAlpha=0, maxAngularAcceleration=10, maxAngularVelocity=2, maxLinearAcceleration=1000, maxLinearVelocity=450, worldAlpha=1}
   }
   n4 = Component To Local(ComponentPose=@n3)
   n5 = Get FE_withpawn() -> (FE_withpawn:bool)
   n6 = RigidBody(ComponentPose=Local To Component(LocalPose=Use cached pose 'Base'())) {
    bClampLinearTranslationLimitToRefPose=true
    bOverrideWorldGravity=true
    overridePhysicsAsset=/BRCosmetics/Weapons/FORT_Melee/Pickaxe_DandyFlare/Meshes/DandyFlare_Axe_Physics
    simulationSpace=BaseBoneSpace
    baseBoneRef {boneName=Melee_TwoHanded_Handle}
    componentAppliedLinearAccClamp=<3000 3000 4000>
    overrideWorldGravity {z=-600}
    simSpaceSettings {dampingAlpha=0, maxAngularAcceleration=15, maxAngularVelocity=3, maxLinearAcceleration=1000, maxLinearVelocity=450, worldAlpha=0.6}
   }
   Output Pose(Result=Blend Poses by bool(
     BlendPose_0=Blend Poses by bool(BlendPose_0=@n4, BlendPose_1=@n2, bActiveValue=@n5)
     BlendPose_1=Component To Local(ComponentPose=@n6)
     bActiveValue=Get bIsInFrontEnd()
    )) {layerGroup=None}
   AnimGraphNode_SubInput_0 = Input Pose()
   Save cached pose 'Base'(Pose=@AnimGraphNode_SubInput_0) {cacheName=Base}
  ]}
 EventGraph {nodes [
   Event BlueprintUpdateAnimation()
   > Set FE_withpawn(FE_withpawn:bool=AND Boolean(A=Get bIsInFrontEnd(), B=Get bIsOnPawn() -> (bIsOnPawn:bool))) -> (Output_Get:bool)
   TryGetPawnOwner()
  ]}
]
variables [FE_withpawn {type=boolean}]
